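# Default Lit material for the trim of the organ pipes environment prop, flagged for use with Nanite.
# It holds three fixed texture samples from the trim set: the D texture, a normal map and an S texture sampled as masks.
# The graph has no parameters or math nodes, and shows no links from the samples to material outputs.
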Material /SparksMaps/Environment/Props/Organ/Materials/M_OrganPipes_Trim
graph {
 properties {bUsedWithNanite=true, pixelDepthOffsetMode=PDOM_Legacy, usedShadingModels=Default Lit}
 nodes [
  TextureSample() {texture=/SparksMaps/Environment/Props/Organ/Textures/T_OrganPipes_Trim_D}
  TextureSample() {samplerType=SAMPLERTYPE_Normal, texture=/SparksMaps/Environment/Props/Organ/Textures/T_OrganPipes_Trim_N}
  TextureSample() {samplerType=SAMPLERTYPE_Masks, texture=/SparksMaps/Environment/Props/Organ/Textures/T_OrganPipes_Trim_S}
 ]
}
