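# Material for the screen on an organ prop, producing a CRT-style pixel pattern that darkens toward the edges.
# It samples a borrowed Jumbotron CRT texture with heavily tiled UVs for a fine, repeating screen pattern.
# LinearGradient U and V values and their inverses are raised to a high power, summed, inverted and clamped into a procedural border mask that multiplies the pattern.
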Material /SparksMaps/Environment/Props/Organ/Materials/M_OrganScreen
graph {
 properties {bUsedWithNanite=true, pixelDepthOffsetMode=PDOM_Legacy, usedShadingModels=Default Lit}
 nodes [
  Constant()
  Constant() {r=1}
  Constant() {r=0.1}
  n1 = TextureSample(UVs=Multiply(A=TextureCoordinate(), B=Constant() {r=256})) {
   samplerSource=SSM_Wrap_WorldGroupSettings
   samplerType=SAMPLERTYPE_LinearColor
   texture=/Game/Accessories/FORT_Backpacks/Backpack_M_MED_Jumbotron_S20/Textures/T_Jumbotron_CRT
  }
  n2 = MaterialFunctionCall() {materialFunction=&LinearGradient}
  n3 = Constant() {r=35}
  n4 = OneMinus(None=@n2.UGradient)
  n5 = OneMinus(None=@n2.VGradient)
  n6 = Add(
   A=Add(A=Power(Base=@n2.UGradient, Exp=@n3) {constExponent=100}, B=Power(Base=@n4, Exp=@n3) {constExponent=100})
   B=Add(A=Power(Base=@n2.VGradient, Exp=@n3) {constExponent=100}, B=Power(Base=@n5, Exp=@n3) {constExponent=100})
  )
  Multiply(A=@n1.RGB, B=Clamp(None=OneMinus(None=@n6)))
 ]
}
