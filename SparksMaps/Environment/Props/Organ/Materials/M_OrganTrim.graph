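# Masked, default-lit, Nanite-enabled material for the trim of an organ prop, built from diffuse, normal and mask trim textures.
# The mask texture's alpha is multiplied by a constant warm golden color, then by an Emissive Intensity scalar parameter to form a tunable glow term.
# The diffuse and normal textures are sampled directly, with no further processing.
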Material /SparksMaps/Environment/Props/Organ/Materials/M_OrganTrim
graph {
 properties {bUsedWithNanite=true, blendMode=BLEND_Masked, pixelDepthOffsetMode=PDOM_Legacy, usedShadingModels=Default Lit}
 nodes [
  Multiply(
   A=Multiply(
    A=TextureSample() {samplerType=SAMPLERTYPE_Masks, texture=/SparksMaps/Environment/Props/Organ/Textures/T_Organ_Trim_S}.A
    B=Constant3Vector() {constant=<rgba 0.8698 0.5484 0.126 1>}
   )
   B=ScalarParameter() {defaultValue=5, parameterName=Emissive Intensity}
  )
  TextureSample() {texture=/SparksMaps/Environment/Props/Organ/Textures/T_Organ_Trim_D}
  TextureSample() {samplerType=SAMPLERTYPE_Normal, texture=/SparksMaps/Environment/Props/Organ/Textures/T_Organ_Trim_N}
 ]
}
