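# Default Lit, masked-blend material for the tech trim on the Organ environment prop, flagged for use with Nanite.
# It samples four TechTrim textures that, by suffix, appear to be diffuse, emissive, normal and a packed mask map.
# The mask map's green channel is scaled by a Metallic Value parameter, the only exposed control.
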Material /SparksMaps/Environment/Props/Organ/Materials/M_TechTrim
graph {
 properties {bUsedWithNanite=true, blendMode=BLEND_Masked, pixelDepthOffsetMode=PDOM_Legacy, usedShadingModels=Default Lit}
 nodes [
  Multiply(
   A=TextureSample() {samplerType=SAMPLERTYPE_Masks, texture=/SparksMaps/Environment/Props/Organ/Textures/T_TechTrim_S}.G
   B=ScalarParameter() {defaultValue=1, parameterName=Metallic Value}
  )
  TextureSample() {texture=/SparksMaps/Environment/Props/Organ/Textures/T_TechTrim_D}
  TextureSample() {texture=/SparksMaps/Environment/Props/Organ/Textures/T_TechTrim_E}
  TextureSample() {samplerType=SAMPLERTYPE_Normal, texture=/SparksMaps/Environment/Props/Organ/Textures/T_TechTrim_N}
 ]
}
